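# Item detail widget that shows an item's description and quantity in the item details UI.
# On item change it sets text from the tooltip system's combined description, falling back to the item's own, and collapses when empty.
# An optional flag hides it for schematics, weapons, traps and heroes.
# Quantity uses the override or stack count, shows only above one, and refreshes through the bound item-changed event.
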WidgetBlueprint /Game/UI/Foundation/Items/Details/ItemDescriptionDetailWidget
graphs [
 EventGraph {nodes [
   Event HandlePreDifferentItemToDetailSet() {is=title K2Node_Event}
   > Is Valid()
   n1 = Create Event()
   n2 = Get ItemToDetail() -> (ItemToDetail:Fort Item Weak object)
   > Unbind Event from On Item Changed(self:Fort Item Object Reference=@n2, Delegate=@n1) {is=title K2Node_RemoveDelegate}
   Event HandleQuantityOverrideChanged() {is=title K2Node_Event}
   > UpdateQuantityTextBlock()
   Event HandlePostDifferentItemToDetailSet() {is=title K2Node_Event}
   > n3 = Is Valid(InputObject=@n2)
   n4 = Get ShouldShowOnlyIfItemIsSimple() -> (ShouldShowOnlyIfItemIsSimple:bool)
   > n5 = Branch(Condition=OR Boolean(A=NOT Boolean(A=@n4), B=IsItemSimple!() -> (Result:bool)))
   > Bind Event to On Item Changed(self:Fort Item Object Reference=@n2, Delegate=@n1) {is=title K2Node_AddDelegate}
   > n6 = SpawnTooltipContextWithSource!(execute, SourcePlayerController:Player Controller Object Reference=GetOwningPlayer()) -> (then, ReturnValue:Fort Tooltip Context Object Reference)
   > n7 = GetCombinedDescription!(
    execute
    self:Fort Tooltip UIContext Object Reference=Get FortTooltipUIContext!() -> (ReturnValue:Fort Tooltip UIContext Object Reference) {is=title K2Node_GetBlueprintContextEx}
    Object:Object Reference=Get ItemToDetail()
    TooltipContext:Fort Tooltip Context Object Reference=@n6
    bSummaryDescription:bool=false
   ) -> (then, OutDescription:Text, ReturnValue:bool)
   n8 = Select(
    Option 0:Text=GetDescription!(self:Fort Item Object Reference=Get ItemToDetail()) -> (ReturnValue:Text)
    Option 1:Text=@n7
    Index=@n7.ReturnValue
   ) -> (ReturnValue:Text)
   > "SetText (Text)"(self=Get ItemDescriptionTextBlock() -> (ItemDescriptionTextBlock:Common Text Object Reference), InText=@n8)
   > SetVisibility(InVisibility=Select(Option 0:ESlateVisibility Enum=SelfHitTestInvisible, Option 1:ESlateVisibility Enum=Collapsed, Index=TextIsEmpty(InText=@n8)) -> (ReturnValue:ESlateVisibility Enum))
   > UpdateQuantityTextBlock()
   SetVisibility(execute=[@n5.else, @n3.Is Not Valid], InVisibility=Collapsed)
  ]}
 HandleItemChanged {nodes [
   HandleItemChanged!() -> (then, ItemChanged:bool=~, AmmoChanged:bool=~, IngredientsChanged:bool=~) {is=title K2Node_FunctionEntry}
   > UpdateQuantityTextBlock()
  ]}
 IsItemSimple {nodes [
   IsItemSimple() {is=title K2Node_FunctionEntry}
   n1 = GetType!(self:Fort Item Object Reference=Get ItemToDetail()) -> (ReturnValue:EFortItemType Enum=WorldItem)
   n2 = Make Array!(
    "[0]":EFortItemType Enum=Schematic
    "[1]":EFortItemType Enum=Weapon
    "[2]":EFortItemType Enum=WeaponMelee
    "[3]":EFortItemType Enum=WeaponRanged
    "[4]":EFortItemType Enum=WeaponHarvest
    "[5]":EFortItemType Enum=Trap
    "[6]":EFortItemType Enum=Hero
   ) -> (Array:Array of EFortItemType Enums)
   > Return Node!(
    execute
    Result:bool=NOT Boolean(A=Contains Item(TargetArray:Array of EFortItemType Enums=@n2, ItemToFind:"EFortItemType Enum (by ref)"=@n1))
   )
  ]}
 UpdateQuantityTextBlock {nodes [
   UpdateQuantityTextBlock() {is=title K2Node_FunctionEntry}
   > n1 = Is Valid(InputObject=Get ItemToDetail())
   > n2 = Branch(Condition="integer >= integer"!(A:int=Get QuantityOverride() -> (QuantityOverride:int=0), B:int=0) -> (ReturnValue:bool) {is=title K2Node_CallFunction})
   > n3 = Set Quantity(Quantity:int=Get QuantityOverride() -> (QuantityOverride=0)) -> (Output_Get:int=0)
   n4 = Branch(execute=[@n3.then, @n7.then], Condition="integer > integer"(A=Get Quantity() -> (Quantity:int=0), B=1) {is=title K2Node_CallFunction})
   n5 = Get QuantityTextBlock() -> (QuantityTextBlock:Common Text Object Reference)
   > SetVisibility(self=@n5, InVisibility=SelfHitTestInvisible)
   n6 = Get Quantity() -> (Quantity=0)
   > "SetText (Text)"(self=@n5, InText=Format Text!(Format="LOCTEXT(\"Quantity: {value}\")", value:Text="To Text (Integer)"(Value=@n6)) -> (Result))
   SetVisibility(execute=[@n4.else, @n1.Is Not Valid], self=@n5, InVisibility=Collapsed)
   n7 = Set Quantity(execute=@n2.else, Quantity=GetNumInStack!(self:Fort Item Object Reference=Get ItemToDetail()) -> (ReturnValue:int=0)) -> (Output_Get=0)
  ]}
]
variables [ShouldShowOnlyIfItemIsSimple {type=boolean}]
widgets {
 info {inheritedWidgetCount=0, parentClass=/Script/FortniteUI.FortItemDetailElementWidget, rootWidgetClass=&SizeBox, widgetCount=4}
 widgets [
  SizeBox_2 {widgetClassPath=&SizeBox, widget=SizeBox_2}
  VerticalBox_0 {
   widgetClassPath=&VerticalBox
   parent=SizeBox_2
   widget=VerticalBox_0
   slot {class=SizeBoxSlot, properties {padding=<ltrb 16 8 16 12>}}
  }
  QuantityTextBlock {
   widgetClassPath=&CommonTextBlock
   bIsVariable=true
   parent=VerticalBox_0
   widget=QuantityTextBlock
   properties {
    style=/FortUILibrary/Styles/Typography/Foundation/TextStyle-Base-S-B.TextStyle-Base-S-B_C
    text="Quantity: 1,000,000"
    font {
     bForceMonospaced=false
     bMaterialIsStencil=false
     fontMaterial=null
     fontObject=&NotoSans
     letterSpacing=0
     monospacedWidth=1
     size=18
     skewAmount=0
     typefaceFontName=NotoSans-Bold
     outlineSettings {
      bApplyOutlineToDropShadows=false
      bMiteredCorners=false
      bSeparateFillAlpha=false
      outlineMaterial=null
      outlineSize=0
      outlineColor=<rgba 0 0 0 1>
     }
    }
    shadowOffset=<0 0>
   }
   slot {class=VerticalBoxSlot, properties {horizontalAlignment=HAlign_Left}}
  }
  ItemDescriptionTextBlock {
   widgetClassPath=&CommonTextBlock
   bIsVariable=true
   parent=VerticalBox_0
   widget=ItemDescriptionTextBlock
   properties {
    autoWrapText=true
    style=/FortUILibrary/Styles/Typography/Foundation/TextStyle-Base-XS-I-40pc.TextStyle-Base-XS-I-40pc_C
    text="{ItemDescription}"
    colorAndOpacity {colorUseRule=UseColor_Specified, specifiedColor=<rgba 1 1 1 0.4>}
    font {
     bForceMonospaced=false
     bMaterialIsStencil=false
     fontMaterial=null
     fontObject=&NotoSans
     letterSpacing=0
     monospacedWidth=1
     size=16
     skewAmount=0
     typefaceFontName=NotoSans-Italic
     outlineSettings {
      bApplyOutlineToDropShadows=false
      bMiteredCorners=false
      bSeparateFillAlpha=false
      outlineMaterial=null
      outlineSize=0
      outlineColor=<rgba 0 0 0 1>
     }
    }
    shadowOffset=<0 0>
   }
   slot {class=VerticalBoxSlot, properties {size {sizeRule=Fill, value=1}}}
  }
 ]
}
