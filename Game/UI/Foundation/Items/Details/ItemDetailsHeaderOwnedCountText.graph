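# Item detail header widget that shows an "Owned: N" count for the item being inspected.
# When a different item is set, it shows the text for Heroes, Schematics, Workers and Defenders, and collapses it otherwise.
# SetOwnedText queries the store context, counting Workers as survivors and other types by item definition.
# Schematics and crafted weapons tint the text with rarity Color5, while other shown items re-apply the text's current style.
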WidgetBlueprint /Game/UI/Foundation/Items/Details/ItemDetailsHeaderOwnedCountText
graphs [
 EventGraph {nodes [
   Event HandlePostDifferentItemToDetailSet() {is=title K2Node_Event}
   > n1 = Is Valid(InputObject=Get ItemToDetail() -> (ItemToDetail:Fort Item Weak object))
   SetVisibility(execute=[@n3.else, @n1.Is Not Valid], self=Get OwnedCountText() -> (OwnedCountText:Common Text Object Reference), InVisibility=Collapsed)
   n2 = GetType!(self:Fort Item Object Reference=Get ItemToDetail()) -> (ReturnValue:EFortItemType Enum=WorldItem)
   n3 = Branch(
    execute=@n1.Is Valid
    Condition=OR Boolean!(
     A="Equal (Enum)"(A:EFortItemType Enum=@n2, B:EFortItemType Enum=Hero)
     B="Equal (Enum)"(A=@n2, B=Schematic)
     C:bool="Equal (Enum)"(A=@n2, B=Worker)
     D:bool="Equal (Enum)"(A=@n2, B=Defender)
    ) -> (ReturnValue)
   )
   > SetVisibility(self=Get OwnedCountText(), InVisibility=SelfHitTestInvisible)
   > SetOwnedText()
   n4 = Get ItemToDetail()
   > n5 = Branch(Condition=IsSchematicOrCraftedWeapon!(self:Fort Item Object Reference=@n4) -> (ReturnValue:bool))
   > n6 = GetRarityDataForItemDef!(
    execute
    ItemDef:Item Definition Base Object Reference=GetItemDefinition!(self:Fort Item Object Reference=@n4) -> (ReturnValue:Fort Item Definition Object Reference)
   ) -> (then, ReturnValue:Fort Rarity Item Data Structure)
   > SetColorAndOpacity(
    self=Get OwnedCountText()
    InColorAndOpacity=Make Slate Color(SpecifiedColor=Break Fort Rarity Item Data!(FortRarityItemData:"Fort Rarity Item Data Structure (by ref)"=@n6) -> (
      Name:Text
      Color1:Linear Color Structure
      Color2:Linear Color Structure
      Color3:Linear Color Structure
      Color4:Linear Color Structure
      Color5:Linear Color Structure
      Radius:float=~
      Falloff:float=~
      Brightness:float=~
      Roughness:float=~
      Glow:float=~
     ) {is=title K2Node_BreakStruct}.Color5)
   )
   n7 = Get OwnedCountText()
   SetStyle!(
    execute=@n5.else
    self:Common Text Object Reference=@n7
    InStyle:Common Text Style Class Reference=Get Style!(self:Common Text Object Reference=@n7) -> (Style:Common Text Style Class Reference)
   ) -> (then)
  ]}
 SetOwnedText {nodes [
   SetOwnedText() {is=title K2Node_FunctionEntry}
   n1 = Get ItemToDetail()
   > n2 = Branch(Condition="Equal (Enum)"(A=GetType(self=@n1) -> (ReturnValue=WorldItem), B=Worker))
   n3 = Get FortStoreContext!() -> (ReturnValue:Fort Store Context Object Reference) {is=title K2Node_GetBlueprintContextEx}
   n4 = Get ItemToDetail()
   > "SetText (Text)"(
    self=Get OwnedCountText()
    InText=Format Text!(
     Format="LOCTEXT(\"Owned: {Count}\")"
     Count:int=GetOwnedSurvivorCount!(self:Fort Store Context Object Reference=@n3, SurvivorItem:Fort Item Object Reference=@n4) -> (ReturnValue:int=0)
    ) -> (Result)
   )
   n5 = GetItemDefinition(self=Get ItemToDetail())
   n6 = Get FortStoreContext()
   "SetText (Text)"(
    execute=@n2.else
    self=Get OwnedCountText()
    InText=Format Text(
     Format="LOCTEXT(\"Owned: {Count}\")"
     Count=GetOwnedItemCount!(self:Fort Store Context Object Reference=@n6, ItemDefinition:Fort Item Definition Object Reference=@n5) -> (ReturnValue:int=0)
    )
   )
  ]}
]
variables []
widgets {
 info {
  inheritedWidgetCount=0
  parentClass=/Script/FortniteUI.FortItemDetailElementWidget
  rootWidgetClass=&CommonTextBlock
  widgetCount=1
 }
 widgets [OwnedCountText {
   widgetClassPath=&CommonTextBlock
   bIsVariable=true
   widget=OwnedCountText
   properties {
    clipping=OnDemand
    scrollStyle=&TextScrollStyle-Base_C
    style=/FortUILibrary/Styles/Typography/Foundation/TextStyle-Base-XS-B-70pc.TextStyle-Base-XS-B-70pc_C
    text="{OwnedCount}"
    colorAndOpacity {colorUseRule=UseColor_Specified, specifiedColor=<rgba 1 1 1 0.7>}
    font {
     bForceMonospaced=false
     bMaterialIsStencil=false
     fontMaterial=null
     fontObject=&NotoSans
     letterSpacing=0
     monospacedWidth=1
     size=16
     skewAmount=0
     typefaceFontName=NotoSans-Bold
     outlineSettings {
      bApplyOutlineToDropShadows=false
      bMiteredCorners=false
      bSeparateFillAlpha=false
      outlineMaterial=null
      outlineSize=0
      outlineColor=<rgba 0 0 0 1>
     }
    }
    shadowOffset=<0 0>
   }
  }]
}
