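# Item detail header widget that shows an item's rarity name beside its type name, formatted as "| Type".
# Detail-item and compare-item events fetch the detailed item's cosmetic rarity data, which supplies the rarity text and colors.
# A cast of the schematic crafting result to a weapon definition picks the coloring: weapons tint both texts from the rarity palette, while others tint only rarity and reapply the type text's style.
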WidgetBlueprint /Game/UI/Foundation/Items/Details/ItemDetailsHeaderRarityTypeText
graphs [EventGraph {nodes [
   n1 = Event HandleItemToDetailChangedBP() {is=title K2Node_Event}
   n2 = Get!(execute:Exec=[@n1.then, @n10.then, @n11.then]) -> (then:Exec, else:Exec, ItemToDetail:Fort Item Weak object)
   > n3 = GetRarityCosmeticForItem!(execute, Item:Fort Item Object Reference=@n2) -> (then, ReturnValue:EFortRarity Enum=Common)
   > n4 = GetRarityData!(execute, Rarity:EFortRarity Enum=@n3) -> (then, ReturnValue:Fort Rarity Item Data Structure)
   > n5 = Sequence()
   n6 = Break Fort Rarity Item Data!(FortRarityItemData:"Fort Rarity Item Data Structure (by ref)"=@n4) -> (
    Name:Text
    Color1:Linear Color Structure
    Color2:Linear Color Structure
    Color3:Linear Color Structure
    Color4:Linear Color Structure
    Color5:Linear Color Structure
    Radius:float=~
    Falloff:float=~
    Brightness:float=~
    Roughness:float=~
    Glow:float=~
   ) {is=title K2Node_BreakStruct}
   > "SetText (Text)"(self=Get RarityTextBlock() -> (RarityTextBlock:Common Text Object Reference), InText=@n6)
   n7 = Get ItemToDetail() -> (ItemToDetail:Fort Item Weak object)
   > "SetText (Text)"(
    self=Get TypeTextBlock() -> (TypeTextBlock:Common Text Object Reference)
    InText=Format Text!(Format="LOCTEXT(\"| {Type}\")", Type:Text=GetItemTypeName!(self:Fort Item Object Reference=@n7, bUsePlural:bool=false) -> (ReturnValue:Text)) -> (Result)
   )
   > SetVisibility(self=Get TypeTextBlock(), InVisibility=SelfHitTestInvisible)
   n8 = Cast To FortWeaponItemDefinition!(
    execute=@n5.then_1
    Object=GetSchematicCraftingResult!(self:Fort Item Object Reference=Get ItemToDetail()) -> (ReturnValue:Fort Item Definition Object Reference)
   ) -> (then, CastFailed, AsFort Weapon Item Definition:Fort Weapon Item Definition Object Reference) {is=title K2Node_DynamicCast}
   > SetColorAndOpacity(self=Get TypeTextBlock(), InColorAndOpacity=Make Slate Color(SpecifiedColor=@n6.Color5))
   > SetColorAndOpacity(self=Get RarityTextBlock(), InColorAndOpacity=Make Slate Color(SpecifiedColor=@n6.Color3))
   n9 = Get TypeTextBlock()
   SetStyle!(
    execute=@n8.CastFailed
    self:Common Text Object Reference=@n9
    InStyle:Common Text Style Class Reference=Get Style!(self:Common Text Object Reference=@n9) -> (Style:Common Text Style Class Reference)
   ) -> (then)
   > SetColorAndOpacity(self=Get RarityTextBlock(), InColorAndOpacity=Make Slate Color(SpecifiedColor=@n6.Color1))
   n10 = Event HandlePostDifferentItemToDetailSet() {is=title K2Node_Event}
   n11 = Event HandlePostDifferentItemToCompareWithSet() {is=title K2Node_Event}
  ]}]
variables []
widgets {
 info {inheritedWidgetCount=0, parentClass=/Script/FortniteUI.FortItemDetailElementWidget, rootWidgetClass=&HorizontalBox, widgetCount=3}
 widgets [
  HorizontalBox_0 {widgetClassPath=&HorizontalBox, widget=HorizontalBox_0, properties {flowDirectionPreference=Culture}}
  RarityTextBlock {
   widgetClassPath=&CommonTextBlock
   bIsVariable=true
   parent=HorizontalBox_0
   widget=RarityTextBlock
   properties {
    scrollStyle=&TextScrollStyle-Base_C
    style=/FortUILibrary/Styles/Typography/Foundation/TextStyle-Base-XS-B.TextStyle-Base-XS-B_C
    text="{RarityName}"
    font {
     bForceMonospaced=false
     bMaterialIsStencil=false
     fontMaterial=null
     fontObject=&NotoSans
     letterSpacing=0
     monospacedWidth=1
     size=16
     skewAmount=0
     typefaceFontName=NotoSans-Bold
     outlineSettings {
      bApplyOutlineToDropShadows=false
      bMiteredCorners=false
      bSeparateFillAlpha=false
      outlineMaterial=null
      outlineSize=0
      outlineColor=<rgba 0 0 0 1>
     }
    }
    shadowOffset=<0 0>
   }
   slot {class=HorizontalBoxSlot}
  }
  TypeTextBlock {
   widgetClassPath=&CommonTextBlock
   bIsVariable=true
   parent=HorizontalBox_0
   widget=TypeTextBlock
   properties {
    scrollStyle=&TextScrollStyle-Base_C
    style=/FortUILibrary/Styles/Typography/Foundation/TextStyle-Base-XS-B-70pc.TextStyle-Base-XS-B-70pc_C
    text="| {TypeName}"
    colorAndOpacity {colorUseRule=UseColor_Specified, specifiedColor=<rgba 1 1 1 0.7>}
    font {
     bForceMonospaced=false
     bMaterialIsStencil=false
     fontMaterial=null
     fontObject=&NotoSans
     letterSpacing=0
     monospacedWidth=1
     size=16
     skewAmount=0
     typefaceFontName=NotoSans-Bold
     outlineSettings {
      bApplyOutlineToDropShadows=false
      bMiteredCorners=false
      bSeparateFillAlpha=false
      outlineMaterial=null
      outlineSize=0
      outlineColor=<rgba 0 0 0 1>
     }
    }
    shadowOffset=<0 0>
   }
   slot {class=HorizontalBoxSlot, properties {padding=<ltrb 4 0 0 0>}}
  }
 ]
}
