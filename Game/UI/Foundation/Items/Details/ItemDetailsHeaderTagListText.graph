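# Item detail header widget showing the detailed item's tags, including weapon tier, as a single text line.
# When the detailed or compared item changes, it fetches user-friendly tags and FormatItemDisplayTags joins them with " | ".
# The text block collapses when the result is empty.
# Schematics and crafted weapons tint the text with their rarity's Color5; other or invalid items reapply the text block's current style.
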WidgetBlueprint /Game/UI/Foundation/Items/Details/ItemDetailsHeaderTagListText
graphs [
 EventGraph {nodes [
   n1 = Event HandlePostDifferentItemToDetailSet() {is=title K2Node_Event}
   n2 = FormatItemDisplayTags!(TagTexts:Array of Texts=GetUserFriendlyTags!(
     Item:Fort Item Object Reference=Get ItemToDetail() -> (ItemToDetail:Fort Item Weak object)
     bAddWeaponTier:bool=true
     None:%=Get FortInventoryContext!() -> (ReturnValue:Fort Inventory Context Object Reference) {is=title K2Node_GetBlueprintContextEx}
    ) -> (ReturnValue:Array of Texts)) -> (FormatedText:Text)
   "SetText (Text)"(execute=[@n1.then, @n8.then], self=Get TagListTextBlock() -> (TagListTextBlock:Common Text Object Reference), InText=@n2)
   > SetVisibility(self=Get TagListTextBlock(), InVisibility=Select(Option 0=Visible, Option 1=Collapsed, Index=TextIsEmpty(InText=@n2)))
   n3 = Get ItemToDetail()
   > n4 = Is Valid(InputObject=@n3)
   > n5 = Branch(Condition=IsSchematicOrCraftedWeapon!(self:Fort Item Object Reference=@n3) -> (ReturnValue:bool))
   > n6 = GetRarityDataForItemDef!(
    execute
    ItemDef:Item Definition Base Object Reference=GetItemDefinition!(self:Fort Item Object Reference=@n3) -> (ReturnValue:Fort Item Definition Object Reference)
   ) -> (then, ReturnValue:Fort Rarity Item Data Structure)
   > SetColorAndOpacity(
    self=Get TagListTextBlock()
    InColorAndOpacity=Make Slate Color(SpecifiedColor=Break Fort Rarity Item Data!(FortRarityItemData:"Fort Rarity Item Data Structure (by ref)"=@n6) -> (
      Name:Text
      Color1:Linear Color Structure
      Color2:Linear Color Structure
      Color3:Linear Color Structure
      Color4:Linear Color Structure
      Color5:Linear Color Structure
      Radius:float=~
      Falloff:float=~
      Brightness:float=~
      Roughness:float=~
      Glow:float=~
     ) {is=title K2Node_BreakStruct}.Color5)
   )
   n7 = Get TagListTextBlock()
   SetStyle!(
    execute=[@n5.else, @n4.Is Not Valid]
    self:Common Text Object Reference=@n7
    InStyle:Common Text Style Class Reference=Get Style!(self:Common Text Object Reference=@n7) -> (Style:Common Text Style Class Reference)
   ) -> (then)
   n8 = Event HandlePostDifferentItemToCompareWithSet() {is=title K2Node_Event}
  ]}
 FormatItemDisplayTags {nodes [
   n1 = FormatItemDisplayTags!() -> (then, TagTexts:Array of Texts) {is=title K2Node_FunctionEntry}
   > n2 = For Each Loop(Array:Array of Texts=@n1) -> (Array Element:Text)
   Return Node!(execute=@n2.Completed, FormatedText:Text=Get Result() -> (Result:Text))
   > n3 = Branch(Condition=TextIsEmpty(InText=Get Result()))
   > Set Result(Result:Text=@n2) -> (Output_Get:Text)
   Set Result(execute=@n3.else, Result=Format Text!(Format="LOCTEXT(\"{A} | {B}\")", A:Text=Get Result(), B:Text=@n2) -> (Result))
  ]}
]
variables []
widgets {
 info {
  inheritedWidgetCount=0
  parentClass=/Script/FortniteUI.FortItemDetailElementWidget
  rootWidgetClass=&CommonTextBlock
  widgetCount=1
 }
 widgets [TagListTextBlock {
   widgetClassPath=&CommonTextBlock
   bIsVariable=true
   widget=TagListTextBlock
   properties {
    scrollStyle=&TextScrollStyle-Base_C
    style=/FortUILibrary/Styles/Typography/Foundation/TextStyle-Base-XS-B-70pc.TextStyle-Base-XS-B-70pc_C
    text="{TagList}"
    colorAndOpacity {colorUseRule=UseColor_Specified, specifiedColor=<rgba 1 1 1 0.7>}
    font {
     bForceMonospaced=false
     bMaterialIsStencil=false
     fontMaterial=null
     fontObject=&NotoSans
     letterSpacing=0
     monospacedWidth=1
     size=16
     skewAmount=0
     typefaceFontName=NotoSans-Bold
     outlineSettings {
      bApplyOutlineToDropShadows=false
      bMiteredCorners=false
      bSeparateFillAlpha=false
      outlineMaterial=null
      outlineSize=0
      outlineColor=<rgba 0 0 0 1>
     }
    }
    shadowOffset=<0 0>
   }
  }]
}
