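# StatNumericTextBlock is a stat widget pairing a numeric text block with an arrow image that shows a better, worse or unchanged comparison.
# SetComparisonState stores the buff state and calls UpdateComparison, which also runs on Construct; PreConstruct applies the configurable text style.
# UpdateComparison tints text and arrow with stylesheet-derived base or buff color, hides the arrow for NoChange and rotates it 180 degrees for Worse.
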WidgetBlueprint /Game/UI/Foundation/Stats/StatNumericTextBlock
graphs [
 EventGraph {nodes [
   Event Construct()
   > UpdateComparison()
   Event PreConstruct()
   > SetStyle!(
    execute
    self:Common Text Object Reference=Get NumericText() -> (NumericText:Common Numeric Text Block Object Reference)
    InStyle:Common Text Style Class Reference=Get Numeric Text Style() -> (Numeric Text Style:Common Text Style Class Reference)
   ) -> (then)
  ]}
 SetComparisonState {nodes [
   n1 = SetComparisonState!() -> (then, Buff State:EFortBuffState Enum=NoChange) {is=title K2Node_FunctionEntry}
   > Set Buff State(Buff State:EFortBuffState Enum=@n1) -> (Output_Get:EFortBuffState Enum=NoChange)
   > UpdateComparison()
  ]}
 UpdateComparison {nodes [
   UpdateComparison() {is=title K2Node_FunctionEntry}
   > n1 = Sequence()
   > n2 = "Get Base && Buff Colors"!(
    execute
    Display Type:EFortStatValueDisplayType Enum=BasicSingle
    Buff State:EFortBuffState Enum=Get Buff State() -> (Buff State:EFortBuffState Enum=NoChange)
    Stylesheet:Fort UIStylesheet Structure=Get StyleSheet() -> (StyleSheet:Fort UIStylesheet Structure)
   ) -> (then, Base:Linear Color Structure, Buff:Linear Color Structure) {is=title K2Node_CallFunction}
   > n3 = Sequence()
   n4 = Select!(
    NoChange:Linear Color Structure=@n2
    Better:Linear Color Structure=@n2.Buff
    Worse:Linear Color Structure=@n2.Buff
    Index:EFortBuffState Enum=Get Buff State() -> (Buff State=NoChange)
   ) -> (ReturnValue:Linear Color Structure)
   > SetColorAndOpacity(self=Get NumericText(), InColorAndOpacity=Make Slate Color(SpecifiedColor=@n4))
   SetColorAndOpacity(
    execute=@n3.then_1
    self:Image Object Reference=Get ComparisonArrowImage() -> (ComparisonArrowImage:Image Object Reference)
    InColorAndOpacity:Linear Color Structure=@n4
   )
   n5 = Get Buff State() -> (Buff State=NoChange)
   SetVisibility(
    execute=@n1.then_1
    self=Get ComparisonArrowImage()
    InVisibility=Select(
     NoChange:ESlateVisibility Enum=Collapsed
     Better:ESlateVisibility Enum=HitTestInvisible
     Worse:ESlateVisibility Enum=HitTestInvisible
     Index=@n5
    ) -> (ReturnValue:ESlateVisibility Enum)
   )
   > SetRenderTransformAngle!(
    execute
    self:Widget Object Reference=Get ComparisonArrowImage()
    Angle:float=Select(NoChange:double=0, Better:double=0, Worse:double=180, Index=@n5) -> (ReturnValue:double)
   ) -> (then)
  ]}
]
variables [
 Buff State {type=EFortBuffState, default=NoChange}
 Numeric Text Style {type=Class, default=&TextStyle-Base_C}
 StyleSheet {
  type=FortUIStylesheet
  default {
   color-BoltElemental=<rgba 0.6724 0.0452 0.6724 1>
   color-Buff=<rgba 0.32 1 0.3974 1>
   color-Debuff=<rgba 0.75 0.0939 0.0712 1>
   color-FireElemental=<rgba 0.9301 0.2918 0.013 1>
   color-HarvestWeakPoint=<rgba 0.1276 0.8477 0.875 1>
   color-IceElemental=<rgba 0.016 0.4735 0.8879 1>
   color-Unique=<rgba 0.0999 0.8228 0.491 1>
   item-Icon-Size {Large=<326 326>, Medium=<124 124>, Small=<78 78>, VeryLarge=<202 202>, VerySmall=<46 46>, VeryVerySmall=<30 30>}
   item-Size {Large=<336 336>, Medium=<125 125>, Small=<80 80>, VeryLarge=<208 208>, VerySmall=<48 48>, VeryVerySmall=<32 32>}
   standard-Icon-Size {Large=<104 104>, Medium=<40 40>, Small=<24 24>, VeryLarge=<64 64>, VerySmall=<16 16>, VeryVerySmall=<12 12>}
   theme-Critical {
    active=<rgba 0.2271 0.0638 0.0329 1>
    dark=<rgba 0.0802 0.0046 0.0003 1>
    faded=<rgba 0.325 0.0987 0.0511 1>
    light=<rgba 0.8122 0.148 0.1036 1>
    normal=<rgba 0.1975 0.0152 0.0006 1>
    ultra=<rgba 0.0127 0.0017 0.0012 1>
   }
   theme-Default {
    active=<rgba 0.0004 0.2549 0.1087 1>
    dark=<rgba 0 0.0869 0.0781 1>
    faded=<rgba 0.0461 0.325 0.1836 1>
    light=<rgba 0.0987 0.8277 0.4995 1>
    normal=<rgba 0 0.1701 0.1511 1>
    ultra=<rgba 0.0127 0.0017 0.0012 1>
   }
   theme-Emphasis {
    active=<rgba 1 0.4423 0.0399 1>
    dark=<rgba 0.4591 0.148 0 1>
    faded=<rgba 1 0.4762 0.2388 1>
    light=<rgba 1 0.6253 0.548 1>
    normal=<rgba 0.7519 0.3345 0.0303 1>
    ultra=<rgba 0.0781 0.0415 0 1>
   }
  }
 }
]
widgets {
 info {inheritedWidgetCount=0, parentClass=&CommonUserWidget, rootWidgetClass=&HorizontalBox, widgetCount=3}
 widgets [
  HorizontalBox_0 {widgetClassPath=&HorizontalBox, widget=HorizontalBox_0}
  NumericText {
   widgetClassPath=/Script/CommonUI.CommonNumericTextBlock
   bIsVariable=true
   parent=HorizontalBox_0
   widget=NumericText
   properties {
    style=&TextStyle-Base_C
    text=Text Block
    font {
     bForceMonospaced=false
     bMaterialIsStencil=false
     fontMaterial=null
     fontObject=&NotoSans
     letterSpacing=0
     monospacedWidth=1
     size=24
     skewAmount=0
     typefaceFontName=NotoSans-Regular
     outlineSettings {
      bApplyOutlineToDropShadows=false
      bMiteredCorners=false
      bSeparateFillAlpha=false
      outlineMaterial=null
      outlineSize=0
      outlineColor=<rgba 0 0 0 1>
     }
    }
    renderTransform {angle=0, scale=<1 1>, shear=<0 0>, translation=<0 -2>}
    shadowOffset=<0 0>
   }
   slot {class=HorizontalBoxSlot, properties {verticalAlignment=VAlign_Center}}
  }
  ComparisonArrowImage {
   widgetClassPath=&Image
   bIsVariable=true
   parent=HorizontalBox_0
   widget=ComparisonArrowImage
   properties {
    visibility=Collapsed
    brush {
     bIsDynamicallyLoaded=false
     drawAs=Image
     imageType=NoImage
     mirroring=NoMirror
     resourceName=None
     resourceObject=/Game/UI/Tooltips/Textures/Tooltip-Buff-Arrow
     tiling=NoTile
     imageSize=<16 16>
     margin=<ltrb 0 0 0 0>
     outlineSettings {
      bUseBrushTransparency=false
      roundingType=HalfHeightRadius
      width=0
      color {colorUseRule=UseColor_Specified, specifiedColor=<rgba 0 0 0 0>}
      cornerRadii=<0 0 0 0>
     }
     tintColor {colorUseRule=UseColor_Specified, specifiedColor=<rgba 1 1 1 1>}
     uVRegion {bIsValid=false, max=<0 0>, min=<0 0>}
    }
    renderTransform {angle=0, scale=<1 1>, shear=<0 0>, translation=<0 -2>}
   }
   slot {class=HorizontalBoxSlot, properties {verticalAlignment=VAlign_Center, padding=<ltrb 4 0 0 0>}}
  }
 ]
}
