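# ItemRatingIndicator is a Save the World-aware item rating widget (FortItemRatingIndicator child) showing an icon and numeric power rating.
# On Construct, the bAllStWMoonbeamHUD runtime option selects the layout: text over a translucent backing without the icon, or icon plus text without the backing, adjusting text padding.
# On view model change, it sets the medium-size rating type icon and pushes ComparisonResult and RatingValue into the StatNumericTextBlock.
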WidgetBlueprint /Game/UI/Foundation/Items/Details/ItemRatingIndicator
graphs [EventGraph {nodes [
   Event Construct()
   > n1 = Branch(Condition=Get bAllStWMoonbeamHUD!(self:Fort Runtime Options Object Reference=GetRuntimeOptions()) -> (bAllStWMoonbeamHUD:bool))
   > SetVisibility(self=Get PowerRatingIconImage() -> (PowerRatingIconImage:Image Object Reference), InVisibility=Collapsed)
   n2 = SlotAsOverlaySlot(Widget=Get PowerRatingText() -> (PowerRatingText:Stat Numeric Text Block Object Reference))
   > SetPadding!(execute, self:Overlay Slot Object Reference=@n2, InPadding:Margin Structure=Make Margin(Left=16, Top=0, Right=16, Bottom=0)) -> (then)
   SetVisibility(execute=@n1.else, self=Get Moonbeam_Power_Back_1() -> (Moonbeam_Power_Back_1:Image Object Reference), InVisibility=Collapsed)
   > SetPadding(self=@n2, InPadding=Make Margin(Left=0, Top=0, Right=16, Bottom=0))
   Event HandleViewModelChanged() {is=title K2Node_Event}
   > n3 = Sequence()
   > n4 = GetBrushBySizeFromMultiSizeBrush!(
    execute
    MultiSizeBrush:"Fort Multi Size Brush Structure (by ref)"=Get RatingTypeIcon() -> (RatingTypeIcon:Fort Multi Size Brush Structure)
    Size:EFortBrushSize Enum=Medium
   ) -> (then, ReturnValue:Slate Brush Structure)
   > SetBrush(self=Get PowerRatingIconImage(), InBrush=@n4)
   SetComparisonState!(
    execute=@n3.then_1
    self:Stat Numeric Text Block Object Reference=Get PowerRatingText()
    Buff State:EFortBuffState Enum=Get ComparisonResult() -> (ComparisonResult:EFortBuffState Enum=NoChange)
   ) -> (then)
   > SetCurrentValue!(
    execute
    self:Common Numeric Text Block Object Reference=Get NumericText!(self:Stat Numeric Text Block Object Reference=Get PowerRatingText()) -> (NumericText:Common Numeric Text Block Object Reference)
    NewValue:float="To Float (Integer)"(InInt=Get RatingValue() -> (RatingValue:int=0))
   ) -> (then)
  ]}]
variables []
widgets {
 info {inheritedWidgetCount=0, parentClass=/Script/FortniteUI.FortItemRatingIndicator, rootWidgetClass=&HorizontalBox, widgetCount=6}
 widgets [
  HorizontalBox_0 {widgetClassPath=&HorizontalBox, widget=HorizontalBox_0, properties {flowDirectionPreference=Culture}}
  SizeBox_0 {
   widgetClassPath=&SizeBox
   parent=HorizontalBox_0
   widget=SizeBox_0
   properties {bOverride_HeightOverride=true, bOverride_WidthOverride=true, heightOverride=32, widthOverride=32}
   slot {class=HorizontalBoxSlot, properties {horizontalAlignment=HAlign_Left, verticalAlignment=VAlign_Center}}
  }
  PowerRatingIconImage {
   widgetClassPath=&Image
   bIsVariable=true
   parent=SizeBox_0
   widget=PowerRatingIconImage
   properties {brush {
     bIsDynamicallyLoaded=false
     drawAs=Image
     imageType=NoImage
     mirroring=NoMirror
     resourceName=None
     resourceObject=/Game/UI/Foundation/Textures/Icons/Shell/T-Icon-Power-64
     tiling=NoTile
     imageSize=<40 40>
     margin=<ltrb 0 0 0 0>
     outlineSettings {
      bUseBrushTransparency=false
      roundingType=HalfHeightRadius
      width=0
      color {colorUseRule=UseColor_Specified, specifiedColor=<rgba 0 0 0 0>}
      cornerRadii=<0 0 0 0>
     }
     tintColor {colorUseRule=UseColor_Specified, specifiedColor=<rgba 1 1 1 1>}
     uVRegion {bIsValid=false, max=<0 0>, min=<0 0>}
    }}
   slot {class=SizeBoxSlot}
  }
  Overlay_2 {
   widgetClassPath=&Overlay
   parent=HorizontalBox_0
   widget=Overlay_2
   slot {class=HorizontalBoxSlot, properties {horizontalAlignment=HAlign_Left, verticalAlignment=VAlign_Center, padding=<ltrb 4 0 0 0>}}
  }
  Moonbeam_Power_Back_1 {
   widgetClassPath=&Image
   bIsVariable=true
   parent=Overlay_2
   widget=Moonbeam_Power_Back_1
   properties {
    brush {
     bIsDynamicallyLoaded=false
     drawAs=Image
     imageType=NoImage
     mirroring=NoMirror
     resourceName=None
     resourceObject=/Game/UI/Foundation/Textures/Icons/Items/T-Items-Power-BG
     tiling=NoTile
     imageSize=<40 24>
     margin=<ltrb 0 0 0 0>
     outlineSettings {
      bUseBrushTransparency=false
      roundingType=HalfHeightRadius
      width=0
      color {colorUseRule=UseColor_Specified, specifiedColor=<rgba 0 0 0 0>}
      cornerRadii=<0 0 0 0>
     }
     tintColor {colorUseRule=UseColor_Specified, specifiedColor=<rgba 1 1 1 1>}
     uVRegion {bIsValid=false, max=<0 0>, min=<0 0>}
    }
    colorAndOpacity=<rgba 1 1 1 0.5>
   }
   slot {class=OverlaySlot, properties {horizontalAlignment=HAlign_Fill, verticalAlignment=VAlign_Fill}}
  }
  PowerRatingText {
   widgetClassPath=/Game/UI/Foundation/Stats/StatNumericTextBlock
   bIsVariable=true
   parent=Overlay_2
   widget=PowerRatingText
   properties {numeric Text Style=/FortUILibrary/Styles/Typography/Foundation/TextStyle-Base-M-B.TextStyle-Base-M-B_C}
   slot {class=OverlaySlot, properties {verticalAlignment=VAlign_Center, padding=<ltrb 16 0 16 0>}}
  }
 ]
}
