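# Item count UI widget showing an item's icon, a count and an optional name that collapses when empty.
# Construct and design-time PreConstruct apply item, brush size and text style, with DisplayIconLeftSide choosing the visible icon and count padding.
# SetItemDefinition refreshes both icons and the count, while SetOverrideValue switches the count to an override number.
# SetIsBeingPreviewed toggles a green-tinted border, and SetBrushAndTextStyle resizes icons and restyles the count.
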WidgetBlueprint /Game/UI/Foundation/Items/ItemCount
graphs [
 EventGraph {nodes [
   n1 = Event Construct()
   n2 = Get IconRight() -> (IconRight:Fort Lazy Image Object Reference)
   n3 = Get IconLeft() -> (IconLeft:Fort Lazy Image Object Reference)
   SetBrushFromItemDefinition!(
    execute=[@n1.then, @n15.then]
    self:Fort Lazy Image Object Reference=[@n2, @n3]
    ItemDefinition:Item Definition Base Object Reference=Get ItemDefinition() -> (ItemDefinition:Item Definition Base Object Reference)
   ) -> (then)
   > SetFortBrushSize!(
    execute
    self:Fort Lazy Image Object Reference=[@n2, @n3]
    FortBrushSize:EFortBrushSize Enum=Get BrushSize() -> (BrushSize:EFortBrushSize Enum=VeryVerySmall)
   ) -> (then)
   n4 = Get CountText() -> (CountText:Fort Item Count Text Block Object Reference)
   > SetStyle!(
    execute
    self:Common Text Object Reference=@n4
    InStyle:Common Text Style Class Reference=Get TextStyle() -> (TextStyle:Common Text Style Class Reference)
   ) -> (then)
   > UpdateOverrideState()
   > SetItemDefinition!(execute, self:Fort Item Count Text Block Object Reference=@n4, InItemDefinition:Item Definition Base Object Reference=Get ItemDefinition()) -> (then)
   n5 = Get Name() -> (Name:Text)
   > SetVisibility(
    self=Get DisplayName() -> (DisplayName:Common Text Object Reference)
    InVisibility=Select(Option 0=SelfHitTestInvisible, Option 1=Collapsed, Index=TextIsEmpty(InText=@n5))
   )
   > n6 = Cast To HorizontalBoxSlot!(execute, Object=Get Slot!(self:Widget Object Reference=Get CountText()) -> (Slot:Panel Slot Object Reference)) -> (then, CastFailed, AsHorizontal Box Slot:Horizontal Box Slot Object Reference) {is=title K2Node_DynamicCast}
   n7 = Get DisplayIconLeftSide() -> (DisplayIconLeftSide:bool)
   > n8 = SetPadding!(
    execute
    self:Horizontal Box Slot Object Reference=@n6
    InPadding:Margin Structure=Select(
     Option 0:Margin Structure=Make Margin(Left=10, Top=1, Right=10, Bottom=1)
     Option 1:Margin Structure=Make Margin(Left=10, Top=1, Right=4, Bottom=1)
     Index=@n7
    ) -> (ReturnValue:Margin Structure)
   ) -> (then)
   n9 = Branch(execute=[@n8.then, @n6.CastFailed], Condition=@n7)
   n10 = Get IconLeft()
   > SetVisibility(self=@n10, InVisibility=Visible)
   n11 = Get IconRight()
   > n12 = SetVisibility(self=@n11, InVisibility=Collapsed)
   "SetText (Text)"(execute=[@n13.then, @n12.then], self=Get DisplayName(), InText=Get Name())
   > SetStyle(self=Get DisplayName(), InStyle=Get TextStyle())
   SetVisibility(execute=@n9.else, self=@n10, InVisibility=Collapsed)
   > n13 = SetVisibility(self=@n11, InVisibility=Visible)
   n14 = Event PreConstruct()
   > n15 = Branch(Condition=@n14.IsDesignTime)
   SetBrushAndTextStyle!(execute, self, InBrushSize:EFortBrushSize Enum=VeryVerySmall, InStyle:Common Text Style Class Reference) -> (then)
  ]}
 SetBrushAndTextStyle {nodes [
   n1 = SetBrushAndTextStyle!() -> (then, InBrushSize:EFortBrushSize Enum=VeryVerySmall, InStyle:Common Text Style Class Reference) {is=title K2Node_FunctionEntry}
   > SetFortBrushSize(self=[Get IconRight(), Get IconLeft()], FortBrushSize=@n1)
   > SetStyle(self=Get CountText(), InStyle=@n1.InStyle)
  ]}
 SetIsBeingPreviewed {nodes [
   n1 = SetIsBeingPreviewed!() -> (then, BeingPreviewed:bool=~) {is=title K2Node_FunctionEntry}
   > SetVisibility(
    self=Get Border() -> (Border:Common Border Object Reference)
    InVisibility=Select(Option 0:ESlateVisibility Enum=Hidden, Option 1:ESlateVisibility Enum=SelfHitTestInvisible, Index=@n1) -> (ReturnValue:ESlateVisibility Enum)
   )
  ]}
 SetItemDefinition {nodes [
   n1 = SetItemDefinition!() -> (then, ItemDefinition:Item Definition Base Object Reference) {is=title K2Node_FunctionEntry}
   > n2 = Set ItemDefinition(ItemDefinition:Item Definition Base Object Reference=@n1) -> (Output_Get:Item Definition Base Object Reference)
   > SetBrushFromItemDefinition(self=[Get IconLeft(), Get IconRight()], ItemDefinition=@n2)
   > SetItemDefinition(self=Get CountText(), InItemDefinition=@n2) {is=title K2Node_CallFunction}
  ]}
 SetOverrideValue {nodes [
   n1 = SetOverrideValue!() -> (then, OverrrideValue:int) {is=title K2Node_FunctionEntry}
   > Set CountStyle(CountStyle:EFortItemCountStyle Enum=OverrideCount) -> (Output_Get:EFortItemCountStyle Enum=StackCount)
   > Set OverrideValue(OverrideValue:int=@n1) -> (Output_Get:int=0)
   > UpdateOverrideState()
  ]}
 UpdateOverrideState {nodes [
   UpdateOverrideState() {is=title K2Node_FunctionEntry}
   n1 = Get CountStyle() -> (CountStyle:EFortItemCountStyle Enum=StackCount)
   > SetCountStyle!(execute, self:Fort Item Count Text Block Object Reference=Get CountText(), InCountStyle:EFortItemCountStyle Enum=@n1) -> (then)
   > Branch(Condition=OR Boolean(
     A="Equal (Enum)"(A:EFortItemCountStyle Enum=@n1, B:EFortItemCountStyle Enum=OverrideCount)
     B="Equal (Enum)"(A=@n1, B=StackCountOverOverride)
    ))
   > SetOverrideValue!(execute, self:Fort Item Count Text Block Object Reference=Get CountText(), InOverrideValue:int=Get OverrideValue() -> (OverrideValue:int=0)) -> (then) {is=title K2Node_CallFunction}
  ]}
]
variables [
 BrushSize {type=EFortBrushSize, default=Medium}
 CountStyle {type=EFortItemCountStyle, default=StackCount}
 DisplayIconLeftSide {type=boolean}
 ItemDefinition {type=ItemDefinitionBase, default=null}
 Name {type=string, default=""}
 OverrideValue {type=integer}
 TextStyle {type=Class, default=null}
]
widgets {
 info {inheritedWidgetCount=0, parentClass=&CommonUserWidget, rootWidgetClass=&Overlay, widgetCount=7}
 widgets [
  Overlay_0 {widgetClassPath=&Overlay, widget=Overlay_0}
  Border {
   widgetClassPath=&CommonBorder
   bIsVariable=true
   parent=Overlay_0
   widget=Border
   properties {
    visibility=Hidden
    background {
     bIsDynamicallyLoaded=false
     drawAs=Box
     imageType=NoImage
     mirroring=NoMirror
     resourceName=None
     resourceObject=null
     tiling=NoTile
     imageSize=<32 32>
     margin=<ltrb 0 0 0 0>
     outlineSettings {
      bUseBrushTransparency=false
      roundingType=HalfHeightRadius
      width=0
      color {colorUseRule=UseColor_Specified, specifiedColor=<rgba 0 0 0 0>}
      cornerRadii=<0 0 0 0>
     }
     tintColor {colorUseRule=UseColor_Specified, specifiedColor=<rgba 1 1 1 1>}
     uVRegion {bIsValid=false, max=<0 0>, min=<0 0>}
    }
    brushColor=<rgba 0.1981 0.5583 0.1812 1>
    contentColorAndOpacity=<rgba 0.1981 0.5583 0.1812 1>
   }
   slot {class=OverlaySlot, properties {horizontalAlignment=HAlign_Fill, verticalAlignment=VAlign_Fill}}
  }
  HorizontalBox_0 {
   widgetClassPath=&HorizontalBox
   parent=Overlay_0
   widget=HorizontalBox_0
   properties {flowDirectionPreference=Culture}
   slot {class=OverlaySlot}
  }
  IconLeft {
   widgetClassPath=&FortLazyImage
   bIsVariable=true
   parent=HorizontalBox_0
   widget=IconLeft
   slot {class=HorizontalBoxSlot, properties {horizontalAlignment=HAlign_Center, verticalAlignment=VAlign_Center}}
  }
  CountText {
   widgetClassPath=/Script/FortniteUI.FortItemCountTextBlock
   bIsVariable=true
   parent=HorizontalBox_0
   widget=CountText
   properties {
    itemCountMaxInterpolateDuration=1
    text=Text Block
    font {
     bForceMonospaced=false
     bMaterialIsStencil=false
     fontMaterial=null
     fontObject=&NotoSans
     letterSpacing=0
     monospacedWidth=1
     size=24
     skewAmount=0
     typefaceFontName=NotoSans-Regular
     outlineSettings {
      bApplyOutlineToDropShadows=false
      bMiteredCorners=false
      bSeparateFillAlpha=false
      outlineMaterial=null
      outlineSize=0
      outlineColor=<rgba 0 0 0 1>
     }
    }
   }
   slot {class=HorizontalBoxSlot, properties {verticalAlignment=VAlign_Center, padding=<ltrb 10 1 10 1>}}
  }
  IconRight {
   widgetClassPath=&FortLazyImage
   bIsVariable=true
   parent=HorizontalBox_0
   widget=IconRight
   slot {class=HorizontalBoxSlot, properties {horizontalAlignment=HAlign_Center, verticalAlignment=VAlign_Center}}
  }
  DisplayName {
   widgetClassPath=&CommonTextBlock
   bIsVariable=true
   parent=HorizontalBox_0
   widget=DisplayName
   properties {
    style=&TextStyle-Base_C
    text=Text Block
    font {
     bForceMonospaced=false
     bMaterialIsStencil=false
     fontMaterial=null
     fontObject=&NotoSans
     letterSpacing=0
     monospacedWidth=1
     size=24
     skewAmount=0
     typefaceFontName=NotoSans-Regular
     outlineSettings {
      bApplyOutlineToDropShadows=false
      bMiteredCorners=false
      bSeparateFillAlpha=false
      outlineMaterial=null
      outlineSize=0
      outlineColor=<rgba 0 0 0 1>
     }
    }
    shadowOffset=<0 0>
   }
   slot {class=HorizontalBoxSlot, properties {padding=<ltrb 4 4 4 4>}}
  }
 ]
}
