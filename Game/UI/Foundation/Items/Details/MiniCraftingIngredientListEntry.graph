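# Compact crafting ingredient list entry for item details UI, built on FortItemQuantityListEntryBase around a single ItemCount widget.
# Refresh runs on entry refresh, item or quantity changes and preview start/end, setting the item definition and override count.
# It flags the count as previewed when previewing, the preview quantity differs and inspect mode is Overview (compared twice identically).
# A tooltip function creates a name and description tooltip for valid items, except under gamepad input.
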WidgetBlueprint /Game/UI/Foundation/Items/Details/MiniCraftingIngredientListEntry
graphs [
 EventGraph {nodes [
   Event RefreshListEntry() {is=title K2Node_Event}
   > Refresh()
   Event HandleDifferentItemOrQuantitySetBP!() -> (OutputDelegate, then, IsBeingReset:bool) {is=title K2Node_Event}
   > Refresh()
   Event PreviewStarted() {is=title K2Node_Event}
   > Refresh()
   Event PreviewEnded() {is=title K2Node_Event}
   > Refresh()
  ]}
 Get_ItemCountWidget_ToolTipWidget {nodes [
   Get_ItemCountWidget_ToolTipWidget() {is=title K2Node_FunctionEntry}
   n1 = GetItemDefinition!(self:Fort Item Quantity List Entry Base Object Reference) -> (ReturnValue:Item Definition Base Object Reference)
   > n2 = Is Valid(InputObject=@n1)
   n3 = Get CommonInputSubsystem!() -> (ReturnValue:Common Input Object Reference) {is=title K2Node_GetBlueprintContextEx}
   > n4 = Branch(Condition="Equal (Enum)"(A=GetCurrentInputType(self=@n3), B=Gamepad))
   Return Node!(execute=[@n4.then, @n2.Is Not Valid], ReturnValue:Widget Object Reference)
   n5 = Create Basic Tooltip!(
    execute=@n4.else
    Owning Player:Player Controller Object Reference=GetOwningPlayer()
    Body Text:Text=GetItemShortDescription!(self:Item Definition Base Object Reference=@n1) -> (ReturnValue:Text)
    Header Text:Text=GetDisplayName!(self:Item Definition Base Object Reference=@n1) -> (ReturnValue:Text)
    Icon Brush:Slate Brush Asset Object Reference
    Header Style:Common Text Style Class Reference
    Body Style:Common Text Style Class Reference
   ) -> (then, Output:User Widget Object Reference)
   > Return Node(ReturnValue=@n5)
  ]}
 Refresh {nodes [
   Refresh() {is=title K2Node_FunctionEntry}
   > SetItemDefinition!(
    execute
    self:Item Count Object Reference=Get ItemCountWidget() -> (ItemCountWidget:Item Count Object Reference)
    ItemDefinition:Item Definition Base Object Reference=GetItemDefinition()
   ) -> (then)
   > SetOverrideValue!(execute, self:Item Count Object Reference=Get ItemCountWidget(), OverrrideValue:int=Get Quantity() -> (Quantity:int=0)) -> (then)
   > RefreshBeingPreviewed()
  ]}
 RefreshBeingPreviewed {nodes [
   RefreshBeingPreviewed() {is=title K2Node_FunctionEntry}
   n1 = Get Quantity() -> (Quantity=0)
   n2 = Break Fort Item Entry Preview Data!(FortItemEntryPreviewData:"Fort Item Entry Preview Data Structure (by ref)"=GetPreviewData!(self:Fort Item Quantity List Entry Base Object Reference) -> (OutPreviewData:Fort Item Entry Preview Data Structure)) -> (Quantity:int, InspectMode:EFortItemInspectionMode Enum) {is=title K2Node_BreakStruct}
   n3 = "Equal (Enum)"(A:EFortItemInspectionMode Enum=@n2.InspectMode, B:EFortItemInspectionMode Enum=Overview)
   n4 = "Equal (Enum)"(A=@n2.InspectMode, B=Overview)
   > SetIsBeingPreviewed!(
    execute
    self:Item Count Object Reference=Get ItemCountWidget()
    BeingPreviewed:bool=AND Boolean!(
     A=IsBeingPreviewed!(self:Fort Item Quantity List Entry Base Object Reference) -> (ReturnValue:bool)
     B="Not Equal (Integer)"!(A:int=@n2, B:int=@n1) -> (ReturnValue:bool) {is=title K2Node_CallFunction}
     C:bool=OR Boolean(A=@n3, B=@n4)
    ) -> (ReturnValue)
   ) -> (then)
  ]}
]
variables []
widgets {
 info {
  inheritedWidgetCount=0
  parentClass=/Script/FortniteUI.FortItemQuantityListEntryBase
  rootWidgetClass=/Game/UI/Foundation/Items/ItemCount.ItemCount_C
  widgetCount=1
 }
 widgets [ItemCountWidget {
   widgetClassPath=/Game/UI/Foundation/Items/ItemCount
   bIsVariable=true
   widget=ItemCountWidget
   properties {
    countStyle=OverrideCount
    displayIconLeftSide=true
    textStyle=/FortUILibrary/Styles/Typography/Foundation/TextStyle-Base-XS-B.TextStyle-Base-XS-B_C
   }
  }]
}
