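# Widget Blueprint for a compact crafting ingredient list in item details, built on FortItemQuantityListBase.
# Its only widget is a root WrapBox, EntriesWrapBox, so entries flow and wrap onto new lines.
# The overridden AddListEntry event adds each supplied entry widget to the WrapBox, then sets fixed padding on the returned slot.
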WidgetBlueprint /Game/UI/Foundation/Items/Details/MiniCraftingIngredientList
graphs [EventGraph {nodes [
   n1 = Event AddListEntry!() -> (OutputDelegate, then, ListEntry:Fort Item Quantity List Entry Base Object Reference) {is=title K2Node_Event}
   > n2 = AddChildToWrapBox!(
    execute
    self:Wrap Box Object Reference=Get EntriesWrapBox() -> (EntriesWrapBox:Wrap Box Object Reference)
    Content:Widget Object Reference=@n1.ListEntry
   ) -> (then, ReturnValue:Wrap Box Slot Object Reference)
   > SetPadding!(execute, self:Wrap Box Slot Object Reference=@n2, InPadding:Margin Structure=Make Margin(Left=0, Top=8, Right=12, Bottom=8)) -> (then)
  ]}]
variables []
widgets {
 info {inheritedWidgetCount=0, parentClass=/Script/FortniteUI.FortItemQuantityListBase, rootWidgetClass=&WrapBox, widgetCount=1}
 widgets [EntriesWrapBox {widgetClassPath=&WrapBox, bIsVariable=true, widget=EntriesWrapBox, properties {innerSlotPadding=<8 8>}}]
}
