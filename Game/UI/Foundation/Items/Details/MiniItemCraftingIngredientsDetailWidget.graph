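# Compact item detail element that shows a schematic's crafting ingredients in a MiniCraftingIngredientList inside a styled border.
# On construct and whenever the detailed item changes, it resets the list with that item's recipe costs.
# A change of comparison item feeds that item's costs to the list as comparison quantities.
# GetRecipeItemQuantityPairs casts to FortSchematicItem and rebuilds its RecipeCosts as definition-quantity pairs, adding none if the cast fails.
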WidgetBlueprint /Game/UI/Foundation/Items/Details/MiniItemCraftingIngredientsDetailWidget
graphs [
 EventGraph {nodes [
   n1 = Event Construct()
   SetItemDefinitionsAndQuantities!(
    execute=[@n1.then, @n2.then]
    self:Fort Item Quantity List Base Object Reference=Get MiniCraftingIngredientList() -> (MiniCraftingIngredientList:Mini Crafting Ingredient List Object Reference)
    ItemQuantityPairs:Array of Fort Item Quantity Pair Structures=GetRecipeItemQuantityPairs!(Item:Fort Item Object Reference=Get ItemToDetail() -> (ItemToDetail:Fort Item Weak object)) -> (Pairs:Array of Fort Item Quantity Pair Structures)
    ShouldResetWidgets:bool=true
   ) -> (then)
   Event HandlePostDifferentItemToCompareWithSet() {is=title K2Node_Event}
   > SetItemToCompareDefinitionsAndQuantities!(
    execute
    self:Fort Item Quantity List Base Object Reference=Get MiniCraftingIngredientList()
    ItemToCompareQuantityPairs:Array of Fort Item Quantity Pair Structures=GetRecipeItemQuantityPairs(Item=Get ItemToCompareWith() -> (ItemToCompareWith:Fort Item Weak object))
   ) -> (then)
   n2 = Event HandlePostDifferentItemToDetailSet() {is=title K2Node_Event}
  ]}
 GetRecipeItemQuantityPairs {nodes [
   n1 = GetRecipeItemQuantityPairs!() -> (then, Item:Fort Item Object Reference) {is=title K2Node_FunctionEntry}
   > n2 = Cast To FortSchematicItem!(execute, Object=@n1) -> (then, CastFailed, AsFort Schematic Item:Fort Schematic Item Object Reference) {is=title K2Node_DynamicCast}
   > n3 = For Each Loop(Array:Array of Fort Item Quantity Pair Structures=Break Recipe!(Recipe:"Recipe Structure (by ref)"=GetSchematicRecipe!(self:Fort Schematic Item Object Reference=@n2) -> (ReturnValue:Recipe Structure)) -> (
     RecipeResults:Array of Fort Item Quantity Pair Structures
     bIsConsumed:bool=~
     RecipeCosts:Array of Fort Item Quantity Pair Structures
     RequiredCatalysts:Gameplay Tag Container Structure
     Score:int
    ) {is=title K2Node_BreakStruct}.RecipeCosts) -> (Array Element:Fort Item Quantity Pair Structure)
   > n4 = MakeItemQuantityPair!(
    execute
    ItemDefinition:Item Definition Base Object Reference=GetItemDefinitionFromItemQuantityPair!(ItemQuantityPair:"Fort Item Quantity Pair Structure (by ref)"=@n3) -> (ReturnValue:Item Definition Base Object Reference)
    ItemQuantity:int=Break Fort Item Quantity Pair!(FortItemQuantityPair:"Fort Item Quantity Pair Structure (by ref)"=@n3) -> (Quantity:int) {is=title K2Node_BreakStruct}
   ) -> (then, ReturnValue:Fort Item Quantity Pair Structure)
   > Add(
    TargetArray:Array of Fort Item Quantity Pair Structures=Get Results() -> (Results:Array of Fort Item Quantity Pair Structures)
    NewItem:"Fort Item Quantity Pair Structure (by ref)"=@n4
   ) {is=title K2Node_CallArrayFunction}
   Return Node!(execute=[@n3.Completed, @n2.CastFailed], Pairs:Array of Fort Item Quantity Pair Structures=Get Results())
  ]}
]
variables []
widgets {
 info {inheritedWidgetCount=0, parentClass=/Script/FortniteUI.FortItemDetailElementWidget, rootWidgetClass=&SizeBox, widgetCount=3}
 widgets [
  SizeBox_0 {widgetClassPath=&SizeBox, widget=SizeBox_0}
  CommonBorder_2 {
   widgetClassPath=&CommonBorder
   parent=SizeBox_0
   widget=CommonBorder_2
   properties {style=/FortUILibrary/Styles/Border/Border_StatRow_DarkBlue.Border_StatRow_DarkBlue_C, padding=<ltrb 16 0 16 0>}
   slot {class=SizeBoxSlot}
  }
  MiniCraftingIngredientList {
   widgetClassPath=&/MiniCraftingIngredientList
   bIsVariable=true
   parent=CommonBorder_2
   widget=MiniCraftingIngredientList
   slot {class=BorderSlot, properties {padding=<ltrb 16 0 16 0>}}
  }
 ]
}
